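# Niagara system for the Scifi Streak vehicle drift-trail cosmetic, with tire meshes and trailing ribbons.
# DriftTire burst-spawns mesh particles that follow skeletal wheel bones and scale to wrap each wheel.
# Whip_Trail spawns two quality-scaled ribbons with rearward velocity, fading inherited movement and three layered renderers.
# Both emitters use per-wheel tire-FX disable masks and painted-color variant overrides; ribbons also mask values by game context.
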NiagaraSystem /VehicleCosmetics/DriftTrail/DriftTrail_Scifistreak/FX/NS_DriftTrail_ScifiStreak_MP
dependencies {
 usedDataInterfaces [&NiagaraDataInterfaceCurve, &NiagaraDataInterfaceArrayColor, &NiagaraDataInterfaceSkeletalMesh]
 usedDynamicInputs []
 usedModules [
  &self:SystemSpawnScript.SystemScriptSource.SystemScriptGraph.NiagaraNodeAssignment_0.SetVariables_#1
  &self:SystemSpawnScript.SystemScriptSource.SystemScriptGraph.NiagaraNodeAssignment_1.SetVariables_#2
  &FN_UnpackVehicleFlags
  /VehicleCosmetics/FX/Niagara/Modules/VC_SetVehicleSpeedAttribs
  /Game/Effects/Niagara/Modules/Vehicle/FN_VehicleFlags_Cheatcodes
  &SystemState
 ]
}
emitters {
 DriftTire_Emitter_Ready002 {
  emitterSpawnScript [/SetVariables(Emitter.DotsColor:LinearColor=&SelectLinearColorFromArray(
     Array Sampling Mode=::NewEnumerator1
     Direct Array Index=$User.ColorVariantID
     Color Selection Array=DataInterface() {propertyValues {ColorData [
        <rgba 50 22.3108 0 1>
        <rgba 1 1 1 1>
        <rgba 0 0 0 0>
        <rgba 0 0 0 0>
        <rgba 0 0 0 0>
        <rgba 0 0 0 0>
        <rgba 0 0 0 0>
        <rgba 0 0 0 0>
        <rgba 0 0 0 0>
        <rgba 0 0 0 0>
        <rgba 0 0 0 0>
        <rgba 0 0 0 0>
        <rgba 0 0 0 0>
        <rgba 0 0 0 0>
        <rgba 0 0 0 0>
       ]}}
     Random Seed=?
    )) {isSetParameters=true}]
  emitterUpdateScript [
   &EmitterState(
    Inactive Response=::NewEnumerator1
    Life Cycle Mode=::NewEnumerator1
    Loop Behavior=::NewEnumerator1
    Loop Duration Mode=::NewEnumerator1
    UseLoopDelay=false
   )
   &SpawnBurst_Instantaneous(Spawn Count=2, Use Spawn Probability=false, UseLoopCountLimit=false) {module=SpawnBurst_Instantaneous001}
   &SpawnBurst_Instantaneous(Spawn Count=2, Spawn Group=1, Use Spawn Probability=false, UseLoopCountLimit=false)
   &SpawnBurst_Instantaneous(Spawn Count=2, Spawn Time=7.5167, Use Spawn Probability=false, UseLoopCountLimit=false) {module=SpawnBurst_Instantaneous002}
  ]
  particleSpawnScript [
   &VC_PaintedVariantsOverride(
    Black=<rgba 0.05 0.05 0.05 0.5>
    BurntSienna=<rgba 0.12 0.0508 0.0117 1>
    Cobalt=<rgba 0.05 0.1074 0.5 1>
    Crimson=<rgba 0.25 0 0.0064 1>
    DefaultColor=&MakeCustomLinearColorFromBool(Bool=$User.bIsPainted, True Color=$User.PaintedColor, False Color=<rgba 0 0.2762 0.35 1>)
    ForestGreen=<rgba 0.0118 0.25 0.0069 1>
    Grey=<rgba 0.11 0.11 0.11 1>
    Lime=<rgba 0.125 0.25 0 1>
    Orange=<rgba 0.25 0.1279 0.011 1>
    Pink=<rgba 0.22 0.0447 0.1569 1>
    Purple=<rgba 0.0817 0.015 0.3 1>
    R1 Black:LinearColor=<rgba 1 1 1 1>
    R1 BurntSienna:LinearColor=<rgba 5 2.1263 0.5 1>
    R1 Cobalt:LinearColor=<rgba 2 3.0213 10 1>
    R1 Crimson:LinearColor=<rgba 15 0 0.3814 1>
    R1 DefaultColor=<rgba 0 71.033 90 1>
    R1 ForestGreen:LinearColor=<rgba 0.4706 10 0.2745 1>
    R1 Grey:LinearColor=<rgba 2 2 2 1>
    R1 Lime:LinearColor=<rgba 2.5 5 0 1>
    R1 Orange:LinearColor=<rgba 5 2.5585 0.2194 1>
    R1 Pink:LinearColor=<rgba 5 1.0156 3.567 1>
    R1 Purple:LinearColor=<rgba 1.9064 0.35 7 1>
    R1 Saffron:LinearColor=<rgba 8 6.9677 0.086 1>
    R1 SkyBlue:LinearColor=<rgba 1.0467 3.4392 8 1>
    R1 TitaniumWhite:LinearColor=<rgba 8 8 8 1>
    Saffron=<rgba 0.22 0.1916 0.0024 1>
    SkyBlue=<rgba 0.0484 0.1591 0.37 1>
    TitaniumWhite=<rgba 0.25 0.25 0.25 1>
    Use R1 Painted Overrides=true
   )
   &InitializeParticle(
    Color Mode=::NewEnumerator0
    Mesh Scale Mode=::NewEnumerator4
    Sprite Rotation Mode=::NewEnumerator2
    Sprite Size Mode=::NewEnumerator3
    UsePositionOffset=false
    Write Material Random=true
   )
   &PartitionParticles()
   &SkeletalMeshLocation(
    Bone / Socket Index=&Add_Integer(A=$Particles.PartitionParticles.Partition, B=4)
    Bone / Socket Sampling Mode=::NewEnumerator5
    Interpolate Mesh Sampling=false
    MaskSpawnGroup=false
    Mesh Sampling Type=::NewEnumerator6
    Skeletal Mesh=$System.SkeletalMesh
   )
   /SetVariables(Particles.VisibilityTag=&SpawnGroup()) {isSetParameters=true}
  ]
  particleUpdateScript [
   &FN_DisableTireFX(
    Mask Index Attribute:NiagaraInt32=$Particles.PartitionParticles.Partition
    Mask Index FL=2
    Mask Index FR=3
    Mask Index RL=0
    Mask Index RR=1
   )
   &VC_WheelWrapScale(Wheel Index Attribute:NiagaraInt32=&Add_Integer(A=$Particles.PartitionParticles.Partition, B=2))
   &ParticleState(Kill Particles When Lifetime Has Elapsed=false)
   &UpdateMeshOrientation(Rotation Coordinate Space=::NewEnumerator2, Rotation Vector=<0 1 0>) {enabled=false}
   &SkeletalMeshLocation(
    Bone / Socket Index=&Add_Integer(A=$Particles.PartitionParticles.Partition, B=4)
    Bone / Socket Sampling Mode=::NewEnumerator5
    MaskSpawnGroup=false
    Mesh Sampling Type=::NewEnumerator6
    Skeletal Mesh=$System.SkeletalMesh
   ) {module=SkeletalMeshLocation001}
   &SolveForcesAndVelocity(Clamp Velocity=false, Limit Acceleration=false)
  ]
  rendererClasses [&NiagaraMeshRendererProperties]
  renderers [
   {rendererClass=&NiagaraMeshRendererProperties, rendererIndex=0}
   {rendererClass=&NiagaraMeshRendererProperties, rendererIndex=1}
  ]
 }
 Whip_Trail {
  emitterUpdateScript [
   &EmitterState()
   &VC_RateQualityHandler(Epic Spawn Rate=90, High Spawn Rate=60, Low Spawn Rate=30, Medium Spawn Rate=45, Use Custom Rate?=true)
   &VC_SpawnRibbon_v1(Number of Ribbons=2, Spawn Time=-1, SpawnRate=$Emitter.Spawn Rate)
  ]
  particleSpawnScript [
   &VC_PaintedVariantsOverride(
    Black=<rgba 0.009 0.009 0.009 7>
    BurntSienna=<rgba 0.21 0.083 0.0098 3>
    Cobalt=<rgba 0.0392 0.098 0.5 6>
    Crimson=<rgba 0.28 0 0.0071 4>
    DefaultColor=&MakeCustomLinearColorFromBool(Bool=$User.bIsPainted, True Color=$User.PaintedColor, False Color=<rgba 0.0357 0.7 0.7 1.3>)
    ForestGreen=<rgba 0.0759 0.7 0.0192 3>
    Grey=<rgba 0.05 0.05 0.05 2.5>
    Lime=<rgba 0.175 0.35 0 4>
    Orange=<rgba 0.5 0.2398 0 3>
    Pink=<rgba 0.8196 0.051 0.5412 1.5>
    Purple=<rgba 0.1061 0.0316 0.35 3.5>
    R1 Black:LinearColor=<rgba 0.02 0.02 0.02 1>
    R1 BurntSienna:LinearColor=<rgba 0.6 0.2372 0.0279 0.9>
    R1 Cobalt:LinearColor=<rgba 0.1176 0.2941 1.5 0.7>
    R1 Crimson:LinearColor=<rgba 1 0 0.0254 0.3>
    R1 DefaultColor=<rgba 0.0153 0.3 0.3 1.3>
    R1 ForestGreen:LinearColor=<rgba 0.1691 1.2 0.0329 1>
    R1 Grey:LinearColor=<rgba 0.2 0.2 0.2 0.4>
    R1 Lime:LinearColor=<rgba 0.75 1.5 0 0.4>
    R1 Orange:LinearColor=<rgba 1 0.3488 0 1>
    R1 Pink:LinearColor=<rgba 1.2 0.0746 0.7923 0.6>
    R1 Purple:LinearColor=<rgba 0.2426 0.0723 0.8 0.4>
    R1 Saffron:LinearColor=<rgba 1.2 0.92 0 0.8>
    R1 SkyBlue:LinearColor=<rgba 0.1963 0.6449 1.5 0.8>
    R1 TitaniumWhite:LinearColor=<rgba 0.75 0.75 0.75 0.7>
    R2 DefaultColor:LinearColor=<rgba 0.051 1 1 1.3>
    R2 Lime:LinearColor=<rgba 0.2943 0.5885 0 0.6>
    R2 Saffron:LinearColor=<rgba 0.9 0.69 0 0>
    R3 Black:LinearColor=<rgba 1 1 1 0.1>
    R3 BurntSienna:LinearColor=<rgba 1 0.3088 0.1233 0.7>
    R3 Cobalt:LinearColor=<rgba 0.4 0.4296 2 0.5>
    R3 Crimson:LinearColor=<rgba 1 0 0 1>
    R3 DefaultColor:LinearColor=<rgba 0.196 0.7 0.7 1.3>
    R3 ForestGreen:LinearColor=<rgba 0.0144 0.487 0.0298 2>
    R3 Lime:LinearColor=<rgba 0.2 0.4 0 5>
    R3 Pink:LinearColor=<rgba 0.5052 0.0605 0.3435 0.8>
    R3 Saffron:LinearColor=<rgba 1 0.7667 0 0.6>
    R3 TitaniumWhite:LinearColor=<rgba 0.75 0.75 0.75 0.4>
    Saffron=<rgba 0.3 0.2622 0 3.8>
    SkyBlue=<rgba 0.1098 0.3608 0.8392 2>
    TitaniumWhite=<rgba 1 1 1 1>
    Use R1 Painted Overrides=true
    Use R2 Painted Overrides=true
    Use R3 Painted Overrides=true
   )
   &VC_InitializeRibbon_v1(
    Color=$Particles.Color
    Custom Offset=true
    Enable Skeletal Position=true
    Lifetime=/VehicleCosmetics/FX/Niagara/Modules/VC_Mask_Floats(Coro Value:NiagaraFloat=0.3, FrontEnd Value:NiagaraFloat=0.5, FN Value:NiagaraFloat=0.5)
    Local Vector Offset=&Add_Vector(
     A=&AddVectorToPosition(Position=<40 0 0>, Vector=$User.offset)
     B=&VC_Mask_Vectors(Coro Value:Vector3f=<-40 0 -40>, FrontEnd Value:Vector3f=<0 0 0>, FN Value:Vector3f=<0 0 0>)
    )
    Offset Amount=1
    Ribbon Width=40
   )
   &AddVelocity(
    Velocity=&VC_Mask_Vectors(Coro Value:Vector3f=<-600 0 0>, FrontEnd Value:Vector3f=<-1000 0 0>, FN Value:Vector3f=<-1000 0 0>)
    Velocity Speed Scale=3
   ) {module=AddVelocity001}
   /SetVariables(
    Particles.ColorTag3:LinearColor=&MakeCustomLinearColorFromBool(Bool=$User.bIsPainted, True Color=$User.PaintedColor, False Color=<rgba 0.102 0.2 1 1>)
    Particles.RibbonWidthPanner=100
   ) {isSetParameters=true}
   /VehicleCosmetics/FX/VC_FXParents/Modules/DLR_ProductPreviews(GameModeID:NiagaraInt32=$System.GameModeID, bPreviewVehicleSpeed:NiagaraBool=$System.bPreviewVehicleSpeed) {enabled=false}
  ]
  particleUpdateScript [
   &FN_DisableTireFX(Mask Index Attribute:NiagaraInt32=$Particles.SpawnGroup, Mask Index FL=2, Mask Index FR=3, Mask Index RL=0, Mask Index RR=1)
   &ScaleColor(
    Scale Alpha=&FloatFromCurve(
     FloatCurve=DataInterface() {propertyValues {Curve {keys [{interpMode=RCIM_Cubic, tangentMode=RCTM_Break, time=0.1, value=1}, {interpMode=RCIM_Cubic, tangentMode=RCTM_Break, time=0.8, value=0}]}}}
     Scale Curve=1.7
    )
    Scale RGB=&MakeCustomVectorFromBool(
     Bool=$User.bIsFrontendPreview
     True Vector=&Multiply_VectorByFloat(Vector=<0.5 0.5 0.5>, Float=$User.PaintedScalar)
     False Vector=&Multiply_VectorByFloat(Vector=<1 1 1>, Float=$User.PaintedScalar)
    )
    ScaleA=true
    ScaleRGB=true
   )
   &InheritSourceMovement(
    Applied Position Scale=&Multiply_VectorByFloat(
     Vector=<1 1 1>
     Float=&FloatFromCurve(FloatCurve=DataInterface() {propertyValues {Curve {keys [
          {interpMode=RCIM_Cubic, leaveTangent=-0.0051, tangentMode=RCTM_Break, time=0, value=1}
          {arriveTangent=-0.247, interpMode=RCIM_Cubic, tangentMode=RCTM_Break, time=1, value=0.9}
         ]}}})
    )
    ApplyForce=false
    ApplyPosition=true
    ApplyPositionFalloff=false
   ) {enabled=false, module=InheritSourceMovement001}
   &ParticleState()
   &InheritSourceMovement(
    Applied Position Falloff Distance=650
    Applied Position Scale=&VectorFromFloat(Value=&FloatFromCurve(FloatCurve=DataInterface() {propertyValues {Curve {
         defaultValue=1
         keys [
          {arriveTangent=-1.7885, interpMode=RCIM_Cubic, leaveTangent=-1.7885, tangentMode=RCTM_User, time=0.05, value=1}
          {interpMode=RCIM_Cubic, tangentMode=RCTM_User, time=1, value=0}
         ]
        }}}))
    ApplyForce=false
    ApplyPosition=true
    ApplyPositionFalloff=false
   )
   /SetVariables(
    Particles.Ribbon3Color:LinearColor=&VC_DynamicScaleColor(
     Color to Scale:LinearColor=$Particles.Initial.ColorTag3
     RGB Scale:Vector3f=&MakeCustomVectorFromBool(
      Bool=$User.bIsFrontendPreview
      True Vector=&Multiply_VectorByFloat(Vector=<2 2 2>, Float=$User.PaintedScalar)
      False Vector=&Multiply_VectorByFloat(Vector=<2 2 2>, Float=$User.PaintedScalar)
     )
     Alpha Scale:NiagaraFloat=1
    )
    Particles.Ribbon3Index0:Vector4f=&MakeVector4(
     X=&MakeCustomFloatFromBool(
      Bool=$User.bIsFrontendPreview
      True Float=&Multiply_Float(A=20, B=1)
      False Float=&Multiply_Float(A=20, B=1)
     )
     Y=&FloatFromCurve(FloatCurve=DataInterface() {propertyValues {Curve {keys [<at 0 1>,<at 0.35 5>,<at 0.4 60>]}}})
     Z=0
     W=1
    )
    Particles.Ribbon3Width:NiagaraFloat=&Multiply_Float(A=24, B=1.7)
   ) {isSetParameters=true}
   &SolveForcesAndVelocity(Clamp Velocity=false, Limit Acceleration=false)
   &VC_UpdateRibbon_v1(Enable Skeletal Position=true)
  ]
  rendererClasses [&NiagaraRibbonRendererProperties]
  renderers [
   {rendererClass=&NiagaraRibbonRendererProperties}
   {rendererClass=&NiagaraRibbonRendererProperties, rendererIndex=1}
   {rendererClass=&NiagaraRibbonRendererProperties, rendererIndex=2}
  ]
 }
}
summary {
 systemName=NS_DriftTrail_ScifiStreak_MP
 emitters [
  {emitterName=Whip_Trail, rendererClasses [&NiagaraRibbonRendererProperties]}
  {emitterName=DriftTire_Emitter_Ready002, rendererClasses [&NiagaraMeshRendererProperties]}
 ]
 userVariables [
  User.bIsFrontendPreview {type=NiagaraBool}
  User.bIsPainted {type=NiagaraBool}
  User.bPreviewVehicleSpeed {type=NiagaraBool}
  User.ColorVariantID {type=NiagaraInt32}
  User.GameModeID {type=NiagaraInt32, defaultValue=3}
  User.offset {type=Vector3f, description="offset from wheels(set through cosmetic template)", defaultValue {struct=Vector3f, value=<0 0 0>}}
  User.PaintedColor {type=LinearColor, description="", defaultValue {struct=LinearColor, value=<rgba 1 1 1 1>}}
  User.PaintedScalar {type=NiagaraFloat, defaultValue=1, description=""}
  User.ScalarPainted {type=NiagaraFloat, defaultValue=1, description=""}
  User.VehicleFlags {type=NiagaraInt32, defaultValue=806289408}
  User.Wheel_Radius_Array {
   type=NiagaraDataInterfaceArrayFloat
   defaultValue {
    struct=NiagaraExt_VariableValue_DataInterface
    value {
     dataInterface=&self:NiagaraDataInterfaceArrayFloat_1
     dataInterfaceClass=&NiagaraDataInterfaceArrayFloat
    }
   }
  }
  User.Wheel_Width_Array {
   type=NiagaraDataInterfaceArrayFloat
   defaultValue {
    struct=NiagaraExt_VariableValue_DataInterface
    value {
     dataInterface=&self:NiagaraDataInterfaceArrayFloat_0
     dataInterfaceClass=&NiagaraDataInterfaceArrayFloat
    }
   }
  }
  User.WheelSize {type=NiagaraFloat, defaultValue=0, description=""}
  User.WheelTouchingArray {
   type=NiagaraDataInterfaceArrayFloat
   defaultValue {
    struct=NiagaraExt_VariableValue_DataInterface
    value {
     dataInterface=&self:NiagaraDataInterfaceArrayFloat_2
     dataInterfaceClass=&NiagaraDataInterfaceArrayFloat
    }
   }
  }
 ]
}
